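# Physics asset for the right back flap of the Dieselpunk 02 medium female player body, covering the flap and nearby upper-body bones.
# Capsules approximate the upper spine, right clavicle and right upper arm, with an extra thin, wide box on the clavicle.
# The box-shaped dyn_backFlap_r body is constrained to the clavicle with much tighter swing and twist limits than most of the spine-to-arm chain.
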PhysicsAsset /BRCosmetics/Characters/Player/Female/Medium/Bodies/F_MED_Dieselpunk_02/Meshes/F_MED_Dieselpunk_02_RBackflap_MESH_Physics
bodies {
 clavicle_r [
  {shapeType=Capsule, length=14.1667, radius=4.934, center=<-4.5258 1.5406 1.0762>, rotation=<rot 0 90 -90>}
  {shapeType=Box, extentX=19.2524, extentY=1.8406, extentZ=25.3909, center=<-9.5191 5.9859 2.0128>, rotation=<rot 0 9.2966 9.0377>}
 ]
 dyn_backFlap_r [{shapeType=Box, extentX=1.0622, extentY=8.7518, extentZ=5.5171, center=<4.8111 -0.0168 -0.181>, rotation=<rot -4.0009 -90 1.3739>}]
 spine_04 [{shapeType=Capsule, length=12.5247, radius=5.7071, center=<2.4515 3.1625 9.0981>, rotation=<rot -36.317 -115.1932 38.4611>}]
 spine_05 [{shapeType=Capsule, length=22.6525, radius=6.4117, center=<2.1679 -2.6285 0>, rotation=<rot 0 -90 0>}]
 upperarm_r [{shapeType=Capsule, length=17.0022, radius=5.3003, center=<-10.4584 0.6694 -0.1524>, rotation=<rot -0.8459 -91.7891 -87.1447>}]
}
constraints [
 {bone1Name=spine_05, bone2Name=spine_04, swing1LimitDegrees=45, swing2LimitDegrees=45, twistLimitDegrees=2}
 {bone1Name=clavicle_r, bone2Name=spine_05, swing1LimitDegrees=45, swing2LimitDegrees=45, twistLimitDegrees=45}
 {bone1Name=upperarm_r, bone2Name=clavicle_r, swing1LimitDegrees=45, swing2LimitDegrees=45, twistLimitDegrees=45}
 {bone1Name=dyn_backFlap_r, bone2Name=clavicle_r, swing1LimitDegrees=30, swing2LimitDegrees=5, twistLimitDegrees=2}
]
