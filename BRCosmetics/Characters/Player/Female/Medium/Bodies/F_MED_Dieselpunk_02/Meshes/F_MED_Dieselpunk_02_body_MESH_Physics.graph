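# Physics asset for the body mesh of the female medium Dieselpunk 02 player cosmetic, covering the upper torso and its dangling costume pieces.
# Capsules on upper spine, clavicles and upper arms, plus a box on the left clavicle, form the body shapes.
# Small box and sphere bodies on dyn_ flap and fringe bones hang from the clavicles, spine_05 or, for the top fringe, the front flap.
# Constraints are mostly angle-limited, while left clavicle and front flap rotate freely.
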PhysicsAsset /BRCosmetics/Characters/Player/Female/Medium/Bodies/F_MED_Dieselpunk_02/Meshes/F_MED_Dieselpunk_02_body_MESH_Physics
bodies {
 clavicle_l [
  {shapeType=Capsule, length=12.2302, radius=5.8888, center=<6.3578 -1.7878 -0.3258>, rotation=<rot 1.0792 101.2176 84.5731>}
  {shapeType=Box, extentX=21.9958, extentY=3.901, extentZ=30.3173, center=<9.0383 -4.6915 -0.3286>, rotation=<rot 17.0961 12.1123 8.3523>}
 ]
 clavicle_r [{shapeType=Capsule, length=14.1667, radius=4.934, center=<-4.5258 1.5406 1.0762>, rotation=<rot 0 90 -90>}]
 dyn_backFlap_l [{shapeType=Box, extentX=1.266, extentY=9.1807, extentZ=6.6848, center=<-4.9018 -0.0876 0.2813>, rotation=<rot -6.1189 -89.9999 8.0906>}]
 dyn_frontFlap_r [{shapeType=Sphere, radius=1.1755}]
 dyn_lowFringe_r [{shapeType=Box, extentX=1.0805, extentY=3.611, extentZ=1.6051, center=<2.2298 -0.531 0>, rotation=<rot 0 -90 0>}]
 dyn_topFringe_r [{shapeType=Box, extentX=0.877, extentY=4.1753, extentZ=3.153, center=<2.2976 -0.4739 0.0394>, rotation=<rot -0.064 -89.8991 3.6174>}]
 spine_04 [{shapeType=Capsule, length=12.5247, radius=5.7071, center=<1.5336 2.7908 9.0286>, rotation=<rot -44.7542 -125.0427 44.8886>}]
 spine_05 [{shapeType=Capsule, length=22.6525, radius=6.4117, center=<2.1679 -2.6285 0>, rotation=<rot 0 -90 0>}]
 upperarm_l [{shapeType=Capsule, length=17.0077, radius=5.7759, center=<11.3431 -0.1043 3.1503>, rotation=<rot 0 -90 -90>}]
 upperarm_r [{shapeType=Capsule, length=17.0022, radius=5.3003, center=<-10.4517 0.8837 -0.1492>, rotation=<rot -0.8459 -91.7891 -86.4594>}]
}
constraints [
 {bone1Name=spine_05, bone2Name=spine_04, swing1LimitDegrees=45, swing2LimitDegrees=45, twistLimitDegrees=2}
 {bone1Name=upperarm_l, bone2Name=clavicle_l, swing1LimitDegrees=45, swing2LimitDegrees=45, twistLimitDegrees=45}
 {bone1Name=clavicle_r, bone2Name=spine_05, swing1LimitDegrees=45, swing2LimitDegrees=45, twistLimitDegrees=45}
 {bone1Name=upperarm_r, bone2Name=clavicle_r, swing1LimitDegrees=45, swing2LimitDegrees=45, twistLimitDegrees=45}
 {
  bone1Name=clavicle_l
  bone2Name=spine_05
  swing1LimitDegrees=45
  swing1Motion=Free
  swing2LimitDegrees=45
  swing2Motion=Free
  twistLimitDegrees=45
  twistMotion=Free
 }
 {bone1Name=dyn_backFlap_l, bone2Name=clavicle_l, swing1LimitDegrees=30, swing2LimitDegrees=5, twistLimitDegrees=2}
 {bone1Name=dyn_lowFringe_r, bone2Name=spine_05, swing1LimitDegrees=30, swing2LimitDegrees=30, twistLimitDegrees=2}
 {bone1Name=dyn_topFringe_r, bone2Name=dyn_frontFlap_r, swing1LimitDegrees=30, swing1Motion=Locked, swing2LimitDegrees=30, twistLimitDegrees=2}
 {
  bone1Name=dyn_frontFlap_r
  bone2Name=clavicle_r
  swing1LimitDegrees=45
  swing1Motion=Free
  swing2LimitDegrees=45
  swing2Motion=Free
  twistLimitDegrees=45
  twistMotion=Free
 }
]
