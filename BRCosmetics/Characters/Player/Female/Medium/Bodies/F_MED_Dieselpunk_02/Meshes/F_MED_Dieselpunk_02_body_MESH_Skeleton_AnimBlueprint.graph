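# Secondary-motion AnimBlueprint for the F_MED_Dieselpunk_02 female body cosmetic, driving its shoulder pad and flaps on top of a pose copied from the attached parent.
# RigidBody simulation feeds bone offsets and rotation constraints blended by skydiving alpha.
# A pistol-alpha offset poses the right back flap, whose second simulation uses the inverse alpha.
# Upper-arm Pose Drivers then adjust flaps, and a Control Rig receives the front-end flag as a float.
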AnimBlueprint /BRCosmetics/Characters/Player/Female/Medium/Bodies/F_MED_Dieselpunk_02/Meshes/F_MED_Dieselpunk_02_body_MESH_Skeleton_AnimBlueprint
target_skeleton=&/F_MED_Dieselpunk_02_body_MESH_Skeleton
graphs [
 AnimGraph {nodes [
   n1 = RigidBody(ComponentPose=Local To Component(LocalPose=Copy Pose From Mesh() {bUseAttachedParent=true})) {
    bOverrideWorldGravity=true
    overridePhysicsAsset=&/F_MED_Dieselpunk_02_body_MESH_Physics
    simulationSpace=BaseBoneSpace
    alphaBoolBlend {blendOption=Linear}
    baseBoneRef {boneName=spine_04}
    componentAppliedLinearAccClamp=<1500 1500 1500>
    componentLinearAccScale=<4 4 4>
    componentLinearVelScale=<4 4 4>
    overrideWorldGravity {z=-980}
   }
   n2 = "Transform (Modify) Bone - Bone: shoulderPad_l"(ComponentPose=@n1, Alpha=1, Translation="0,0,-0.5", Rotation="0,0,0", Scale="1,1,1") {
    rotationMode=BMM_Additive
    translationMode=BMM_Additive
    alphaBoolBlend {blendOption=Linear}
    boneToModify {boneName=shoulderPad_l}
    translation {z=-0.5}
   }
   n3 = Get SkydivingAlpha() -> (SkydivingAlpha:float)
   n4 = "Transform (Modify) Bone - Bone: dyn_backFlapBase_l"(ComponentPose=@n2, Alpha=@n3, Translation="0,-2,0", Rotation="0,0,0", Scale="1,1,1") {
    rotationMode=BMM_Additive
    translationMode=BMM_Additive
    alphaBoolBlend {blendOption=Linear}
    boneToModify {boneName=dyn_backFlapBase_l}
    translation {y=-2}
   }
   n5 = "Transform (Modify) Bone - Bone: dyn_topFlapBase_l"(ComponentPose=@n4, Alpha=@n3, Translation="0.5,1,0", Rotation="0,0,0", Scale="1,1,1.5") {
    rotationMode=BMM_Additive
    scaleMode=BMM_Replace
    scaleSpace=BCS_BoneSpace
    translationMode=BMM_Additive
    translationSpace=BCS_BoneSpace
    alphaBoolBlend {blendOption=Linear}
    alphaScaleBiasClamp {bInterpResult=true, interpSpeedDecreasing=3, interpSpeedIncreasing=3}
    boneToModify {boneName=dyn_topFlapBase_l}
    scale {z=1.5}
    translation=<0.5 1>
   }
   n6 = "Constraint - dyn_topFlapBase_l (RR)"!(ComponentPose=@n5, Alpha=@n3, ConstraintWeights_0=0.5, ConstraintWeights_1:float=0.5) -> (Pose) {
    alphaBoolBlend {blendOption=Linear}
    alphaScaleBiasClamp {bInterpResult=true, interpSpeedDecreasing=4, interpSpeedIncreasing=4}
    boneToModify {boneName=dyn_topFlapBase_l}
    constraintSetup [{transformType=Rotation, targetBone {boneName=upperarm_l}}, {transformType=Rotation, targetBone {boneName=clavicle_l}}]
    constraintWeights [1,1]
   }
   n7 = "Transform (Modify) Bone - Bone: shoulderPad_l"(ComponentPose=@n6, Alpha=@n3, Translation="2,0,-1", Rotation="-15,0,0", Scale="0.8,0.8,0.8") {
    rotationMode=BMM_Additive
    rotationSpace=BCS_BoneSpace
    scaleMode=BMM_Replace
    scaleSpace=BCS_BoneSpace
    translationMode=BMM_Additive
    translationSpace=BCS_BoneSpace
    alphaBoolBlend {blendOption=Linear}
    boneToModify {boneName=shoulderPad_l}
    rotation {pitch=-15}
    scale=<0.8 0.8 0.8>
    translation {x=2, z=-1}
   }
   n8 = "Transform (Modify) Bone - Bone: dyn_frontFlapBase_l"(ComponentPose=@n7, Alpha=@n3, Translation="2,0,-2", Rotation="0,0,0", Scale="1,1,1") {
    translationMode=BMM_Additive
    translationSpace=BCS_BoneSpace
    alphaBoolBlend {blendOption=Linear}
    boneToModify {boneName=dyn_frontFlapBase_l}
    translation {x=2, z=-2}
   }
   n9 = "Constraint - dyn_topFlapBase_r (RR)"!(ComponentPose=@n8, Alpha=@n3, ConstraintWeights_0=0.5, ConstraintWeights_1:float=0.5) -> (Pose) {
    alphaBoolBlend {blendOption=Linear}
    alphaScaleBiasClamp {bInterpResult=true, interpSpeedDecreasing=4, interpSpeedIncreasing=4}
    boneToModify {boneName=dyn_topFlapBase_r}
    constraintSetup [{transformType=Rotation, targetBone {boneName=upperarm_r}}, {transformType=Rotation, targetBone {boneName=clavicle_r}}]
    constraintWeights [1,1]
   }
   n10 = "Transform (Modify) Bone - Bone: dyn_topFlapBase_r"(ComponentPose=@n9, Alpha=@n3, Translation="1,-1,1", Rotation="0,0,0", Scale="1,1,1.6") {
    rotationMode=BMM_Additive
    scaleMode=BMM_Replace
    scaleSpace=BCS_BoneSpace
    translationMode=BMM_Additive
    translationSpace=BCS_BoneSpace
    alphaBoolBlend {blendOption=Linear}
    alphaScaleBiasClamp {bInterpResult=true, interpSpeedDecreasing=3, interpSpeedIncreasing=3}
    boneToModify {boneName=dyn_topFlapBase_r}
    scale {z=1.6}
    translation=<1 -1 1>
   }
   n11 = "Transform (Modify) Bone - Bone: dyn_backFlapBase_r"(ComponentPose=@n10, Alpha=@n3, Translation="0,-0.5,0", Rotation="0,0,0", Scale="1,1,1") {
    rotationMode=BMM_Additive
    scaleMode=BMM_Replace
    scaleSpace=BCS_BoneSpace
    translationMode=BMM_Additive
    translationSpace=BCS_BoneSpace
    alphaBoolBlend {blendOption=Linear}
    alphaScaleBiasClamp {bInterpResult=true, interpSpeedDecreasing=3, interpSpeedIncreasing=3}
    boneToModify {boneName=dyn_backFlapBase_r}
    translation {y=-0.5}
   }
   n12 = "Transform (Modify) Bone - Bone: dyn_backFlapBase_r"(ComponentPose=@n11, Alpha=Get ArmPistolAlpha() -> (ArmPistolAlpha:float), Translation="0,0.5,0", Rotation="0,-15,-10", Scale="1,1,1") {
    rotationMode=BMM_Additive
    scaleMode=BMM_Replace
    scaleSpace=BCS_BoneSpace
    translationMode=BMM_Additive
    translationSpace=BCS_BoneSpace
    alphaBoolBlend {blendOption=Linear}
    alphaScaleBiasClamp {bInterpResult=true}
    boneToModify {boneName=dyn_backFlapBase_r}
    rotation {roll=-10, yaw=-15}
    translation {y=0.5}
   }
   n13 = RigidBody(ComponentPose=@n12, Alpha=Get ArmPistolAlphaInverse() -> (ArmPistolAlphaInverse:float)) {
    bOverrideWorldGravity=true
    overridePhysicsAsset=&/F_MED_Dieselpunk_02_RBackflap_MESH_Physics
    simulationSpace=BaseBoneSpace
    alphaBoolBlend {blendOption=Linear}
    alphaScaleBiasClamp {bInterpResult=true}
    baseBoneRef {boneName=spine_04}
    componentAppliedLinearAccClamp=<1000 1000 1000>
    componentLinearAccScale=<4 4 4>
    componentLinearVelScale=<4 4 4>
    overrideWorldGravity {z=-980}
   }
   n14 = Pose Driver - Source: upperarm_r!(SourcePose:Pose Link Structure=Component To Local(ComponentPose=@n13)) -> (Pose:Pose Link Structure) {
    is=title AnimGraphNode_PoseDriver
    poseAsset=&/F_MED_Dieselpunk_02_PistolAim_PoseAsset
    onlyDriveBones [{boneName=dyn_frontFlapBase_r}]
    poseTargets [
     {
      bApplyCustomCurve=true
      bIsHidden=false
      distanceMethod=DefaultMethod
      drivenName=Base
      functionType=DefaultFunction
      targetScale=1
      boneTransforms [{targetRotation=<rot -58.1985 2.4027 1.3901>, targetTranslation=<-9.8658 1.2136 0.4104>}]
      customCurve {
       defaultValue=3.4028234663852886e+38
       postInfinityExtrap=RCCE_Constant
       preInfinityExtrap=RCCE_Constant
       keys [
        {
         arriveTangent=0
         arriveTangentWeight=0
         interpMode=RCIM_Linear
         leaveTangent=0
         leaveTangentWeight=0
         tangentMode=RCTM_Auto
         tangentWeightMode=RCTWM_WeightedNone
         time=0
         value=0
        }
        {
         arriveTangent=0
         arriveTangentWeight=0
         interpMode=RCIM_Linear
         leaveTangent=0
         leaveTangentWeight=0
         tangentMode=RCTM_Auto
         tangentWeightMode=RCTWM_WeightedNone
         time=1
         value=1
        }
       ]
      }
      targetRotation=<rot 0 0 0>
     }
     {
      bApplyCustomCurve=true
      bIsHidden=false
      distanceMethod=DefaultMethod
      drivenName=Aim
      functionType=DefaultFunction
      targetScale=0.4868
      boneTransforms [{targetRotation=<rot -10.1182 98.9012 -50.4629>, targetTranslation=<-9.8658 1.2136 0.4104>}]
      customCurve {
       defaultValue=3.4028234663852886e+38
       postInfinityExtrap=RCCE_Constant
       preInfinityExtrap=RCCE_Constant
       keys [
        {
         arriveTangent=0
         arriveTangentWeight=0
         interpMode=RCIM_Linear
         leaveTangent=0
         leaveTangentWeight=0
         tangentMode=RCTM_Auto
         tangentWeightMode=RCTWM_WeightedNone
         time=0.004
         value=0.2027
        }
        {
         arriveTangent=0
         arriveTangentWeight=0
         interpMode=RCIM_Linear
         leaveTangent=0
         leaveTangentWeight=0
         tangentMode=RCTM_Auto
         tangentWeightMode=RCTWM_WeightedNone
         time=1
         value=1
        }
       ]
      }
      targetRotation=<rot 0 0 0>
     }
     {
      bApplyCustomCurve=true
      bIsHidden=false
      distanceMethod=DefaultMethod
      drivenName=AimUp
      functionType=DefaultFunction
      targetScale=0.4868
      boneTransforms [{targetRotation=<rot 30.6747 92.319 -79.8877>, targetTranslation=<-9.8658 1.2136 0.4104>}]
      customCurve {
       defaultValue=3.4028234663852886e+38
       postInfinityExtrap=RCCE_Constant
       preInfinityExtrap=RCCE_Constant
       keys [
        {
         arriveTangent=0
         arriveTangentWeight=0
         interpMode=RCIM_Linear
         leaveTangent=0
         leaveTangentWeight=0
         tangentMode=RCTM_Auto
         tangentWeightMode=RCTWM_WeightedNone
         time=0
         value=0
        }
        {
         arriveTangent=0
         arriveTangentWeight=0
         interpMode=RCIM_Linear
         leaveTangent=0
         leaveTangentWeight=0
         tangentMode=RCTM_Auto
         tangentWeightMode=RCTWM_WeightedNone
         time=0.2492
         value=0.4716
        }
        {
         arriveTangent=0
         arriveTangentWeight=0
         interpMode=RCIM_Linear
         leaveTangent=0
         leaveTangentWeight=0
         tangentMode=RCTM_Auto
         tangentWeightMode=RCTWM_WeightedNone
         time=0.4171
         value=0.6607
        }
        {
         arriveTangent=0
         arriveTangentWeight=0
         interpMode=RCIM_Linear
         leaveTangent=0
         leaveTangentWeight=0
         tangentMode=RCTM_Auto
         tangentWeightMode=RCTWM_WeightedNone
         time=1
         value=1
        }
       ]
      }
      targetRotation=<rot 0 0 0>
     }
    ]
    rBFParams {radius=42.4012}
    sourceBones [{boneName=upperarm_r}]
   }
   n15 = Pose Driver - Source: upperarm_l!(SourcePose:Pose Link Structure=@n14) -> (Pose:Pose Link Structure) {
    is=title AnimGraphNode_PoseDriver
    poseAsset=&/F_MED_Dieselpunk_02_PistolAim_l_PoseAsset
    onlyDriveBones [{boneName=dyn_topFlap_l}, {boneName=dyn_frontFlap_l}]
    poseTargets [
     {
      bApplyCustomCurve=false
      bIsHidden=false
      distanceMethod=DefaultMethod
      drivenName=Base_l
      functionType=DefaultFunction
      targetScale=0.8922
      boneTransforms [{targetRotation=<rot -58.1985 2.4027 1.3901>, targetTranslation=<9.8658 -1.2136 -0.4105>}]
      customCurve {defaultValue=3.4028234663852886e+38, postInfinityExtrap=RCCE_Constant, preInfinityExtrap=RCCE_Constant, keys []}
      targetRotation=<rot 0 0 0>
     }
     {
      bApplyCustomCurve=false
      bIsHidden=false
      distanceMethod=DefaultMethod
      drivenName=Aim_l
      functionType=DefaultFunction
      targetScale=0.8922
      boneTransforms [{targetRotation=<rot -36.8603 50.9752 -67.2916>, targetTranslation=<9.8655 -1.2132 -0.4137>}]
      customCurve {defaultValue=3.4028234663852886e+38, postInfinityExtrap=RCCE_Constant, preInfinityExtrap=RCCE_Constant, keys []}
      targetRotation=<rot 0 0 0>
     }
     {
      bApplyCustomCurve=false
      bIsHidden=false
      distanceMethod=DefaultMethod
      drivenName=AimUp_l
      functionType=DefaultFunction
      targetScale=1
      boneTransforms [{targetRotation=<rot 5.0068 43.0966 -104.4519>, targetTranslation=<9.8586 -1.2112 -0.4192>}]
      customCurve {defaultValue=3.4028234663852886e+38, postInfinityExtrap=RCCE_Constant, preInfinityExtrap=RCCE_Constant, keys []}
      targetRotation=<rot 0 0 0>
     }
    ]
    rBFParams {radius=21.2545}
    sourceBones [{boneName=upperarm_l}]
   }
   n16 = Control Rig!(Source=@n15, Alpha, isInFrontEnd:double=Get isInFrontEndFloat() -> (isInFrontEndFloat:double)) -> (Pose) {controlRigAssetReference {blueprintRigClass=&/F_MED_Dieselpunk_02_body_MESH_CtrlRig.F_MED_Dieselpunk_02_body_MESH_CtrlRig_C}}
   Output Pose(Result=@n16) {layerGroup=None}
   Sequence Player 'F_MED_Dieselpunk_02_PistolAim'() {
    bLoopAnimation=false
    playRate=0
    sequence=&/F_MED_Dieselpunk_02_PistolAim
   }
  ]}
 EventGraph {nodes [
   Event BlueprintUpdateAnimation()
   > Sequence!(execute) -> (then_0, then_1, then_2:Exec)
   > Set isInFrontEndFloat(isInFrontEndFloat:double="To Float (Boolean)"(InBool=Get bIsInFrontEnd())) -> (Output_Get:double)
  ]}
]
variables [isInFrontEndFloat {type=number}]
